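# Control Rig for the PocketScrunchie back bling cosmetic that procedurally drives the left and right eye bones of its small limbed creature skeleton.
# In the forward solve, parent constraints make each eye follow its eye_aim bone on Y and Z translation only, with rotation and scale filtered out.
# Two offset nodes then nudge the eyes by small, opposite-signed X translations, propagated to children.
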ControlRigBlueprint /BRCosmetics/Accessories/FORT_Backpacks/Backpack_F_MED_PocketScrunchie/Meshes/F_MED_PocketScrunchie_Pack_CtrlRig
graphs [
 RigVMFunctionLibrary {links [], nodes []}
 RigVMModel {nodes [
   RigUnit_BeginExecution = &RigUnit_BeginExecution()
   &RigUnit_ParentConstraint(
    AdvancedSettings=(InterpolationType=Average,RotationOrderForFilter=XZY)
    Child=(Type=Bone,Name="eye_l")
    Filter=(TranslationFilter=(bX=false,bY=True,bZ=True),RotationFilter=(bX=false,bY=false,bZ=false),ScaleFilter=(bX=false,bY=false,bZ=false))
    Parents=((Item=(Type=Bone,Name="eye_aim_l"),Weight=1))
    Weight=1
    ExecutePin=@RigUnit_BeginExecution.ExecutePin
   )
   > &RigUnit_ParentConstraint(
    AdvancedSettings=(InterpolationType=Average,RotationOrderForFilter=XZY)
    Child=(Type=Bone,Name="eye_r")
    Filter=(TranslationFilter=(bX=false,bY=True,bZ=True),RotationFilter=(bX=false,bY=false,bZ=false),ScaleFilter=(bX=false,bY=false,bZ=false))
    Parents=((Item=(Type=Bone,Name="eye_aim_r"),Weight=1))
    Weight=1
   )
   > &RigUnit_OffsetTransformForItem(
    Item=(Type=Bone,Name="eye_r")
    OffsetTransform=(Rotation=(X=0,Y=0,Z=0,W=1),Translation=(X=0.2,Y=0,Z=0),Scale3D=(X=1,Y=1,Z=1))
    Weight=1
    bPropagateToChildren=True
   )
   > &RigUnit_OffsetTransformForItem(
    Item=(Type=Bone,Name="eye_l")
    OffsetTransform=(Rotation=(X=0,Y=0,Z=0,W=1),Translation=(X=-0.2,Y=0,Z=0),Scale3D=(X=1,Y=1,Z=1))
    Weight=1
    bPropagateToChildren=True
   )
  ]}
]
hierarchy [
 root {loc=<0 0 0>, rot=<0 0 0 1>}
 pack {parent=root, loc=<0 -0.7316 0>, rot=<0 0 0 1>}
 MSTR_spine_1 {parent=pack, loc=<0 -4.6431 -1.8811>, rot=<0.5039 -0.5039 -0.4961 -0.4961>}
 spine_1 {parent=MSTR_spine_1, loc=<0 0 0>, rot=<0 0 0 1>}
 MSTR_shoulder_r {parent=spine_1, loc=<0.8042 -4.4952 -0.7408>, rot=<0.0087 -0.0276 0.3023 -0.9528>}
 shoulder_r {parent=MSTR_shoulder_r, loc=<0 0 0>, rot=<0 0 0 1>}
 elbow_r {parent=shoulder_r, loc=<4.7816 0 0>, rot=<0.0456 -0.1298 -0.2707 -0.9528>}
 wrist_r {parent=elbow_r, loc=<2.9549 0 0>, rot=<-0.2487 0.0163 -0.1101 -0.9622>}
 MSTR_shoulder_l {parent=spine_1, loc=<0.8042 4.4952 -0.7408>, rot=<0.3023 -0.9528 -0.0087 0.0276>}
 shoulder_l {parent=MSTR_shoulder_l, loc=<0 0 0>, rot=<0 0 0 1>}
 elbow_l {parent=shoulder_l, loc=<-4.7816 0 0>, rot=<0.0456 -0.1298 -0.2707 -0.9528>}
 wrist_l {parent=elbow_l, loc=<-2.9549 0 0>, rot=<-0.2487 0.0163 -0.1101 -0.9622>}
 MSTR_up {parent=spine_1, loc=<-2.0772 0 -1.2296>, rot=<0 -0.9765 0 -0.2156>}
 up {parent=MSTR_up, loc=<0 0 0>, rot=<0 0 0 1>}
 MSTR_eye_aim_r {parent=up, loc=<5.2939 -3.2096 2.9777>, rot=<-0.2259 0.4962 0.2903 -0.7865>}
 eye_aim_r {parent=MSTR_eye_aim_r, loc=<0 0 0>, rot=<0 0 0 1>}
 MSTR_eye_aim_l {parent=up, loc=<5.2939 3.2096 2.9777>, rot=<0.2903 -0.7865 0.2259 -0.4962>}
 eye_aim_l {parent=MSTR_eye_aim_l, loc=<0 0 0>, rot=<0 0 0 1>}
 MSTR_eye_l {parent=up, loc=<6.0891 4.8061 4.5004>, rot=<0.2903 -0.7865 0.2259 -0.4962>}
 eye_l {parent=MSTR_eye_l, loc=<0 0 0>, rot=<0 0 0 1>}
 MSTR_eye_r {parent=up, loc=<6.089 -4.8061 4.5003>, rot=<-0.2259 0.4962 0.2903 -0.7865>}
 eye_r {parent=MSTR_eye_r, loc=<0 0 0>, rot=<0 0 0 1>}
 spine_2 {parent=spine_1, loc=<3.6262 0 0>, rot=<0 -0.0521 0 -0.9986>}
 spine_3 {parent=spine_2, loc=<3.9181 0 0>, rot=<0 0 0 1>}
 leg_r {parent=spine_3, loc=<1.0426 -3.2004 0.1255>, rot=<-0.0066 0.0625 0.1053 -0.9925>}
 knee_r {parent=leg_r, loc=<3.1205 0 0>, rot=<0.0531 -0.3015 -0.0353 -0.9513>}
 toe_r {parent=knee_r, loc=<3.1004 0 0>, rot=<0 0 0 1>}
 leg_l {parent=spine_3, loc=<1.0426 3.2004 0.1255>, rot=<0.1053 -0.9925 0.0066 -0.0625>}
 knee_l {parent=leg_l, loc=<-3.1204 0 0>, rot=<0.0531 -0.3015 -0.0353 -0.9513>}
 toe_l {parent=knee_l, loc=<-3.1004 0 0>, rot=<0 0 0 1>}
]
