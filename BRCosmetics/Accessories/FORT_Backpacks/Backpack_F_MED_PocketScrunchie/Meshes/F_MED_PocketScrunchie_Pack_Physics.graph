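# Physics asset for the PocketScrunchie backpack cosmetic, giving collision and joint limits to the pack and a small limbed figure rigged on it.
# Every body is a single capsule, with a large one covering the pack, which has no constraint.
# Constraints chain the up bone through three spine bones into both legs, while each arm forms a separate chain tied to its MSTR shoulder body.
# Joints allow modest swing with mostly tight twist.
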PhysicsAsset /BRCosmetics/Accessories/FORT_Backpacks/Backpack_F_MED_PocketScrunchie/Meshes/F_MED_PocketScrunchie_Pack_Physics
bodies {
 MSTR_shoulder_l [{shapeType=Capsule, length=1, radius=0.5, shapeName=MSTR_shoulder_l_capsule}]
 MSTR_shoulder_r [{shapeType=Capsule, length=1, radius=0.5, shapeName=MSTR_shoulder_l_capsule, rotation=<rot 0 0 -180>}]
 elbow_l [{shapeType=Capsule, length=0.2041, radius=2.7959, shapeName=elbow_l_capsule, center=<-1.1293 0.405 0>, rotation=<rot 90 180 180>}]
 elbow_r [{shapeType=Capsule, length=0.2041, radius=2.7959, shapeName=elbow_l_capsule, center=<1.1293 -0.405 0>, rotation=<rot 90 0 180>}]
 knee_l [{
   shapeType=Capsule
   length=0.2166
   radius=3.5687
   shapeName=knee_l_capsule
   center=<-1.9853 -0.1547 -0.6515>
   rotation=<rot 79.9999 -179.9999 -180>
  }]
 knee_r [{shapeType=Capsule, length=0.2753, radius=3.13, shapeName=knee_l_capsule, center=<1.4426 -0.0024 1.1264>, rotation=<rot 79.9999 -179.9999 0>}]
 leg_l [{shapeType=Capsule, length=0.4013, radius=2.5987, shapeName=leg_l_capsule, center=<-2.6925 0 -0.1498>, rotation=<rot 90 -179.9802 -179.9803>}]
 leg_r [{shapeType=Capsule, length=0.4013, radius=2.5987, shapeName=leg_l_capsule, center=<1.788 -0.1572 0.1498>, rotation=<rot 90 -144.7356 35.2643>}]
 pack [{shapeType=Capsule, length=6.5395, radius=27.7827, shapeName=pack_capsule, center=<0 25.6708 -5.5037>}]
 shoulder_l [{shapeType=Capsule, length=0.5, radius=3.5, shapeName=shoulder_l_capsule, center=<-1.2936 1.0499 0>, rotation=<rot 90 0 -10>}]
 shoulder_r [{shapeType=Capsule, length=1, radius=3, shapeName=shoulder_l_capsule, center=<1.7501 -0.9694 0>, rotation=<rot 90 0 170>}]
 spine_1 [{shapeType=Capsule, length=1, radius=3, shapeName=spine_1_capsule, rotation=<rot 90 180 180>}]
 spine_2 [{shapeType=Capsule, length=2, radius=4.5, shapeName=spine_2_capsule, center=<-0.0867 0 -0.4919>, rotation=<rot 9.9999 180 180>}]
 spine_3 [{shapeType=Capsule, length=1, radius=5.2, shapeName=spine_3_capsule, center=<-1.5848 0 -0.6776>}]
 toe_l [{shapeType=Capsule, length=1.1493, radius=3.6121, shapeName=toe_l_capsule, center=<-1.4869 0 -0.7868>, rotation=<rot 90 -170.5377 99.4623>}]
 toe_r [{shapeType=Capsule, length=1.2287, radius=3.7812, shapeName=toe_l_capsule, center=<0.6786 0.0873 0.8829>, rotation=<rot 90 180 -90.0001>}]
 up [{shapeType=Capsule, length=5, radius=6.5, shapeName=up_capsule, center=<5.7483 0 0.1371>, rotation=<rot 0.0002 0 90>}]
 wrist_l [{shapeType=Capsule, length=1, radius=3, shapeName=wrist_l_capsule, center=<-1.0288 0.6001 0.1959>, rotation=<rot 90 0.0003 0.0002>}]
 wrist_r [{shapeType=Capsule, length=1, radius=3, shapeName=wrist_l_capsule, center=<1.0288 0 0>, rotation=<rot 90 179.9999 -0.0002>}]
}
constraints [
 {bone1Name=knee_l, bone2Name=toe_l, swing1LimitDegrees=15, swing2LimitDegrees=30, twistLimitDegrees=5}
 {bone1Name=leg_l, bone2Name=knee_l, swing1LimitDegrees=15, swing2LimitDegrees=30, twistLimitDegrees=5}
 {bone1Name=elbow_l, bone2Name=wrist_l, swing1LimitDegrees=30, swing2LimitDegrees=30, twistLimitDegrees=5}
 {bone1Name=shoulder_l, bone2Name=elbow_l, swing1LimitDegrees=30, swing2LimitDegrees=30, twistLimitDegrees=5}
 {bone1Name=spine_1, bone2Name=spine_2, swing1LimitDegrees=25, swing2LimitDegrees=8, twistLimitDegrees=5}
 {bone1Name=spine_2, bone2Name=spine_3, swing1LimitDegrees=25, swing2LimitDegrees=8, twistLimitDegrees=5}
 {bone1Name=spine_3, bone2Name=leg_l, swing1LimitDegrees=15, swing2LimitDegrees=30, twistLimitDegrees=5}
 {bone1Name=up, bone2Name=spine_1, swing1LimitDegrees=20, swing2LimitDegrees=10, twistLimitDegrees=5}
 {bone1Name=leg_r, bone2Name=spine_3, swing1LimitDegrees=15, swing2LimitDegrees=30, twistLimitDegrees=5}
 {bone1Name=knee_r, bone2Name=leg_r, swing1LimitDegrees=15, swing2LimitDegrees=30, twistLimitDegrees=5}
 {bone1Name=toe_r, bone2Name=knee_r, swing1LimitDegrees=15, swing2LimitDegrees=30, twistLimitDegrees=5}
 {bone1Name=shoulder_l, bone2Name=MSTR_shoulder_l, swing1LimitDegrees=30, swing2LimitDegrees=30, twistLimitDegrees=20}
 {bone1Name=shoulder_r, bone2Name=MSTR_shoulder_r, swing1LimitDegrees=35, swing2LimitDegrees=35, twistLimitDegrees=5}
 {bone1Name=elbow_r, bone2Name=shoulder_r, swing1LimitDegrees=35, swing2LimitDegrees=35, twistLimitDegrees=5}
 {bone1Name=wrist_r, bone2Name=elbow_r, swing1LimitDegrees=35, swing2LimitDegrees=35, twistLimitDegrees=5}
]
