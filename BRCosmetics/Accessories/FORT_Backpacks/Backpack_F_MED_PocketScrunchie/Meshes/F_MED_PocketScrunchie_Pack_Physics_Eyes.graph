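# Physics asset for the eye bones of the PocketScrunchie backpack cosmetic, defining bodies and joint limits for the left and right eye aim bones.
# Each side pairs a tiny capsule on its MSTR_eye_aim bone with a larger capsule on eye_aim, offset along X in mirrored directions.
# One constraint per side links the pair, allowing limited swing with almost no twist.
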PhysicsAsset /BRCosmetics/Accessories/FORT_Backpacks/Backpack_F_MED_PocketScrunchie/Meshes/F_MED_PocketScrunchie_Pack_Physics_Eyes
bodies {
 MSTR_eye_aim_l [{shapeType=Capsule, length=0.2, radius=0.2, shapeName=MSTR_eye_aim_l_capsule}]
 MSTR_eye_aim_r [{shapeType=Capsule, length=0.2, radius=0.2, shapeName=MSTR_eye_aim_r_capsule}]
 eye_aim_l [{shapeType=Capsule, length=1, radius=1, shapeName=eye_aim_l_capsule, center=<-3 0 0>}]
 eye_aim_r [{shapeType=Capsule, length=1, radius=1, shapeName=eye_aim_r_capsule, center=<3 0 0>}]
}
constraints [
 {bone1Name=eye_aim_r, bone2Name=MSTR_eye_aim_r, swing1LimitDegrees=20, swing2LimitDegrees=20, twistLimitDegrees=1}
 {bone1Name=eye_aim_l, bone2Name=MSTR_eye_aim_l, swing1LimitDegrees=20, swing2LimitDegrees=20, twistLimitDegrees=1}
]
