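# Anim Blueprint for the PocketScrunchie backpack, adding physics secondary motion and fitting the pack to the wearer.
# The pack bone takes a start offset chosen by body mesh or gender, plus an offset mapped from spine bend.
# That pose feeds front-end or in-game RigidBody sims, blended by bIsInFrontEnd, then an eyes sim and Control Rig.
# The EventGraph sets gravity and acceleration scale from front-end, DBNO and ground state, using pawn velocity as external force.
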AnimBlueprint /BRCosmetics/Accessories/FORT_Backpacks/Backpack_F_MED_PocketScrunchie/Meshes/F_MED_PocketScrunchie_Pack_AnimBP
target_skeleton=&/F_MED_PocketScrunchie_Pack_Skeleton
graphs [
 AnimGraph {nodes [
   n1 = Get gravity_override() -> (gravity_override:Vector)
   n2 = Get PawnVelocity()
   n3 = Get acceleration() -> (acceleration:Vector)
   n4 = RigidBody!(
    ComponentPose=Local To Component(LocalPose=Use cached pose 'Presim'())
    Alpha
    OverrideWorldGravity:Vector=@n1
    ExternalForce=@n2
    ComponentLinearAccScale:Vector=@n3
   ) -> (Pose) {
    bOverrideWorldGravity=true
    overridePhysicsAsset=&/F_MED_PocketScrunchie_Pack_Physics
    alphaBoolBlend {blendOption=Linear}
    baseBoneRef {boneName=pack}
    componentAppliedLinearAccClamp=<3000 3000 3000>
    componentLinearAccScale=<1.2 1.2 1.2>
    componentLinearVelScale=<1 1 1>
    simSpaceSettings {maxAngularAcceleration=3000, maxAngularVelocity=3000, maxLinearAcceleration=3000, maxLinearVelocity=3000}
   }
   n5 = Component To Local(ComponentPose=@n4)
   n6 = RigidBody(
    ComponentPose=Local To Component(LocalPose=Use cached pose 'Presim'())
    OverrideWorldGravity=@n1
    ExternalForce=@n2
    ComponentLinearAccScale=@n3
   ) {
    bOverrideWorldGravity=true
    overridePhysicsAsset=&/F_MED_PocketScrunchie_Pack_FE_Physics
    alphaBoolBlend {blendOption=Linear}
    baseBoneRef {boneName=pack}
    componentAppliedLinearAccClamp=<3000 3000 3000>
    componentLinearAccScale=<1.2 1.2 1.2>
    componentLinearVelScale=<1 1 1>
    simSpaceSettings {maxAngularAcceleration=3000, maxAngularVelocity=3000, maxLinearAcceleration=3000, maxLinearVelocity=3000}
   }
   n7 = Component To Local(ComponentPose=@n6)
   n8 = Get bIsInFrontEnd()
   n9 = RigidBody(
    ComponentPose=Local To Component(LocalPose=Blend Poses by bool(BlendPose_0=@n7, BlendPose_1=@n5, bActiveValue=@n8) {blendType=Linear})
    OverrideWorldGravity=Get gravity_override()
    ExternalForce=Get PawnVelocity()
    ComponentLinearAccScale=Get acceleration()
   ) {
    bOverrideWorldGravity=true
    overridePhysicsAsset=&/F_MED_PocketScrunchie_Pack_Physics_Eyes
    alphaBoolBlend {blendOption=Linear}
    baseBoneRef {boneName=pack}
    componentAppliedLinearAccClamp=<3000 3000 3000>
    componentLinearAccScale=<1.2 1.2 1.2>
    componentLinearVelScale=<1 1 1>
    simSpaceSettings {maxAngularAcceleration=3000, maxAngularVelocity=3000, maxLinearAcceleration=3000, maxLinearVelocity=3000}
   }
   n10 = Control Rig(Source=Component To Local(ComponentPose=@n9)) {
    alphaBoolBlend {blendOption=Linear}
    controlRigAssetReference {blueprintRigClass=&/F_MED_PocketScrunchie_Pack_CtrlRig.F_MED_PocketScrunchie_Pack_CtrlRig_C}
   }
   Output Pose(Result=@n10) {layerGroup=None}
   n11 = Input Pose()
   n12 = Get!() -> (backpack_offset_Location:Vector, backpack_offset_Rotation:Rotator, backpack_offset_Scale:Vector)
   n13 = "Transform (Modify) Bone - Bone: pack"(
    ComponentPose="Transform (Modify) Bone - Bone: pack"(
     ComponentPose=Local To Component(LocalPose=@n11)
     Translation=Get backpack_start_offset() -> (backpack_start_offset:Vector)
     Rotation="0,0,0"
    ) {translationMode=BMM_Additive, alphaBoolBlend {blendOption=Linear}, boneToModify {boneName=pack}}
    Translation=@n12
    Rotation=@n12.backpack_offset_Rotation
   ) {rotationMode=BMM_Additive, translationMode=BMM_Additive, alphaBoolBlend {blendOption=Linear}, boneToModify {boneName=pack}}
   Save cached pose 'Presim'(Pose=Component To Local(ComponentPose=@n13)) {cacheName=Presim}
  ]}
 EventGraph {nodes [
   n1 = Event BlueprintUpdateAnimation()
   n2 = Branch(execute=[@n1.then, @n18.then], Condition=Get bIsInFrontEnd())
   > Set gravity_override(gravity_override="0,0,-980")
   > Set acceleration(acceleration:Vector="0.2,0.2,0.2") -> (Output_Get:Vector)
   n3 = Get bIsDBNO()
   n4 = Branch(execute=@n2.else, Condition=@n3)
   > n5 = Set gravity_override(gravity_override="0,0,-980")
   n6 = GetDeltaTransformFromRefPose!(self=GetOwnerSkeletalMeshComponent(), BoneName=spine_05, BaseName=spine_01) -> (ReturnValue_Location:Vector, ReturnValue_Rotation:Rotator, ReturnValue_Scale:Vector)
   Set backpack_offset(
    execute=[@n10.then, @n5.then]
    backpack_offset:Transform=Map_Range_Transform!(
     value:double=BreakRotator(InRot=@n6.ReturnValue_Rotation).Yaw
     in_A:double=-15
     in_B:double=-65
     transform_Location:Vector="0,-13,-4"
     transform_Rotation:Rotator="0,0,85"
     transform_Scale:Vector="1,1,1"
    ) -> (out_transform:Transform)
   ) -> (Output_Get:Transform)
   n7 = Get bIsOnGround() -> (bIsOnGround:bool)
   > n8 = Branch(Condition=@n7)
   > Set acceleration(acceleration="1.2,1.2,1.2")
   Set acceleration(execute=@n8.else, acceleration="0.7,0.7,0.7")
   n9 = Gravity Override State Function DEPRECATED(
    execute=@n4.else
    CharacterPartAnimInstance=Self-Reference()
    GameTimeInSeconds=GetGameTimeInSeconds()
    GravityOverrideParams=Get Gravity Override Params()
   )
   > n10 = Set gravity_override(gravity_override=@n9)
   n11 = Event OnCharacterCustomizationLoaded()
   n12 = Get Array Item with EFort Custom Part Type(Array:Array of Custom Character Part Object References=@n11.CustomParts) -> (Output:Custom Character Part Object Reference)
   n13 = GetSkeletalMesh!(self:Custom Character Part Object Reference=@n12) -> (ReturnValue:Skeletal Mesh Object Reference)
   n14 = Branch(
    execute=[@n11.then, @n17.then]
    Condition="Equal (Object)"(A=@n13, B=/BRCosmetics/Characters/Player/Female/Medium/Bodies/F_MED_PocketScrunchie/Meshes/F_MED_PocketScrunchie)
   )
   > Set backpack_start_offset(backpack_start_offset:Vector="0,3.5,0") -> (Output_Get:Vector)
   n15 = Branch(execute=@n14.else, Condition="Equal (Object)"(A=@n13, B=/DefaultCosmeticsAthena/Characters/Meshes/F_MED_Prime))
   > Set backpack_start_offset(backpack_start_offset="0,2.5,0")
   n16 = Branch(
    execute=@n15.else
    Condition="Equal (Enum)"(
     A:EFortCustomGender Enum=Get GenderPermitted!(self:Custom Character Part Object Reference=@n12) -> (GenderPermitted:EFortCustomGender Enum=Invalid)
     B:EFortCustomGender Enum=Male
    )
   )
   > Set backpack_start_offset(backpack_start_offset="0,1.6,0")
   Set backpack_start_offset(execute=@n16.else, backpack_start_offset="0,1.7,0")
   n17 = Event BlueprintInitializeAnimation()
   n18 = Set not_in_FE(not_in_FE:bool=NOT Boolean(A=Get bIsInFrontEnd())) -> (Output_Get:bool)
   OR Boolean(A=@n3, B=Get not_in_FE() -> (not_in_FE:bool))
   OR Boolean(A=@n7, B=Get not_in_FE())
  ]}
 Map_Range_Transform {nodes [
   Map_Range_Transform!() -> (
    then
    value:double=~
    in_A:double=~
    in_B:double=~
    transform_Location_X:double
    transform_Location_Y:double
    transform_Location_Z:double
    transform_Rotation_Roll:float
    transform_Rotation_Pitch:float
    transform_Rotation_Yaw:float
    transform_Scale:Vector
   ) {is=title K2Node_FunctionEntry}
   > Return Node!(
    execute
    out_transform_Location_X:double=MapRangeClamped(
     Value=Get value() -> (value:double)
     InRangeA=Get in_A() -> (in_A:double)
     InRangeB=Get in_B() -> (in_B:double)
     OutRangeB=Get!() -> (transform_Location_X:double, transform_Location_Y:double, transform_Location_Z:double, transform_Rotation:Rotator, transform_Scale:Vector)
    )
    out_transform_Location_Y:double=MapRangeClamped(Value=Get value(), InRangeA=Get in_A(), InRangeB=Get in_B(), OutRangeB=Get().transform_Location_Y)
    out_transform_Location_Z:double=MapRangeClamped(Value=Get value(), InRangeA=Get in_A(), InRangeB=Get in_B(), OutRangeB=Get().transform_Location_Z)
    out_transform_Rotation_Roll:float=MapRangeClamped(
     Value=Get value()
     InRangeA=Get in_A()
     InRangeB=Get in_B()
     OutRangeB=Get!() -> (transform_Location:Vector, transform_Rotation_Roll:float, transform_Rotation_Pitch:float, transform_Rotation_Yaw:float, transform_Scale).transform_Rotation_Roll
    )
    out_transform_Rotation_Pitch:float=MapRangeClamped(Value=Get value(), InRangeA=Get in_A(), InRangeB=Get in_B(), OutRangeB=Get().transform_Rotation_Pitch)
    out_transform_Rotation_Yaw:float=MapRangeClamped(Value=Get value(), InRangeA=Get in_A(), InRangeB=Get in_B(), OutRangeB=Get().transform_Rotation_Yaw)
    out_transform_Scale:Vector="1,1,1"
   )
  ]}
]
variables [
 acceleration {type=Vector}
 backpack_offset {type=Transform, default {location=<0 0 0>, rotation=<rot 0 0 0>, scale=<1 1 1>}}
 backpack_start_offset {type=Vector}
 Gravity Override Params {type=GravityOverrideParamsStruct}
 gravity_override {type=Vector}
 not_in_FE {type=boolean}
]
